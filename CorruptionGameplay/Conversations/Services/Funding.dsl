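; Conversation service task node that lets a player pay the service price into a War Effort funding station acting as the NPC participant.
; After the parent task passes, it sends the price and fund choice index to AddToStationFund, then always advances, as the abort branch is unreachable.
; GatherStaticExtraData packs station funds, goal, names, icons and index into string key/value pairs, detecting two-choice stations by subclass cast and reading their second fund at index+1.
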(fn ExecuteTaskNode
  (params (Context :type Conversation%20Context%20Structure%20%28by%20ref%29))
  (bind _fundchoiceindex (Variables|Default|GetFundChoiceIndex))
  (bind _returnvalue (Utilities|String|ToString(Integer) _fundchoiceindex))
  (bind _conversationcontext (LocalVariables|GetConversationContext))
  (bind (_priceout _returnvalue_2) (CallFunction|GetServicePrice _conversationcontext))
  (bind _returnvalue_3 (Utilities|String|Append "Fund Choice Index:  " _returnvalue))
  (bind _self self)
  (bind _fundchoiceindex_2 (Variables|Default|GetFundChoiceIndex))
  (bind _output_get (LocalVariables|SetConversationContext Context))
  (bind _returnvalue_4 (|CallParentFunction _output_get))
  (bind _returnvalue_5 (Conversation|CanConversationContinue _returnvalue_4))
  (if _returnvalue_5
    (Utilities|FlowControl|Sequence
      (:then_0
        ;[1 1 1] Get Funding Component/Actor to call On Funds Added +
        ; Get Funding Station Num
        ; Could this function + data all live ON the component to keep this more reusable on more actor types?
        (bind _returnvalue_6 (Class|FortConversationContextLibrary|GetParticipantActorNPC (LocalVariables|GetConversationContext)))
        (bind _asbp_station_prop_funding_station (Utilities|Casting|CastToBP_StationProp_FundingStation _returnvalue_6))
        (Utilities|IsValid _asbp_station_prop_funding_station
          (:Is%20Valid
            (Development|PrintString _returnvalue_3)
            (Fortnite|FortShippingLog self _returnvalue_3)
            ;[1 1 1] Just to keep data centralized, pull existing gold bar service price from node. Could be replaced with explicit reference (i.e. table row etc) if desired
            (if _returnvalue_2
              (Fortnite|FortShippingLog _self (Utilities|String|ToString(Text) (Utilities|Text|FormatText "NSLOCTEXT(\"[5CF7BF9D41BC4A14D653228A4AF22D4A]\", \"34B2D6834DD3D807696EB19C3DF3C161\", \"Funding: Calling AddToStationFund for Station {0} FundingAmount: {1} FundChoice: {2}\")" (Utilities|String|ToString(Object) _asbp_station_prop_funding_station) _priceout _fundchoiceindex_2)))
              (Class|BPStationPropFundingStation|AddtoStationFund _asbp_station_prop_funding_station (Math|Float|TruncatetoInteger64 _priceout) _fundchoiceindex_2)
              (LocalVariables|SetSucceeded true)
              (else
                (Fortnite|FortShippingLog _self "Funding: Couldn't find service price!"))))))
      (:then_1
        ;[1 1 1] Hook in Succeeded here once owner is correct actor type / has correct component type
        (if true
          (bind _returnvalue_7 (Conversation|AdvanceConversation (LocalVariables|GetConversationContext)))
          (return _returnvalue_7)
          (else
            (bind _returnvalue_8 (Conversation|AbortConversation (LocalVariables|GetConversationContext)))
            (return _returnvalue_8)))))
    (else
      (return _returnvalue_4))))

(fn GatherStaticExtraData
  (params (Context :type Conversation%20Context%20Structure%20%28by%20ref%29))
  (bind _ioedata (LocalVariables|GetIOEData))
  (bind _ioedata_2 (LocalVariables|GetIOEData))
  (bind _ioedata_3 (LocalVariables|GetIOEData))
  (bind _inoutextradata (|CallParentFunction Context))
  (LocalVariables|SetIOEData _inoutextradata)
  (bind _returnvalue (Class|FortConversationContextLibrary|GetParticipantActorNPC Context))
  (bind _asbp_station_prop_funding_station (Utilities|Casting|CastToBP_StationProp_FundingStation _returnvalue)
    (:then
      (Utilities|FlowControl|Sequence
        (:then_0
          (bind (_conversationnodeparameterpair _ismultichoice) (CallFunction|GetIsSingleorMultiEntry _asbp_station_prop_funding_station))
          (Utilities|Array|Add (LocalVariables|GetIOEData) _conversationnodeparameterpair)
          (LocalVariables|SetIsMultiChoice _ismultichoice))
        (:then_1
          (bind (_conversationnodeparameterpair_0 _conversationnodeparameterpair_1) (CallFunction|GetCurrentFundingAmount _asbp_station_prop_funding_station (LocalVariables|GetIsMultiChoice)))
          (Utilities|Array|Add _ioedata_2 _conversationnodeparameterpair_0)
          (Utilities|Array|Add _ioedata_2 _conversationnodeparameterpair_1))
        (:then_2
          (bind _conversationnodeparameterpair_2 (CallFunction|GetMaxRequiredFundingAmount _asbp_station_prop_funding_station))
          (Utilities|Array|Add (LocalVariables|GetIOEData) _conversationnodeparameterpair_2))
        (:then_3
          (bind (_conversationnodeparameterpair_0_2 _conversationnodeparameterpair_1_2) (CallFunction|GetItemIcons _asbp_station_prop_funding_station (LocalVariables|GetIsMultiChoice)))
          (Utilities|Array|Add _ioedata _conversationnodeparameterpair_0_2)
          (Utilities|Array|Add _ioedata _conversationnodeparameterpair_1_2))
        (:then_4
          (bind (_conversationnodeparameterpair_0_3 _conversationnodeparameterpair_1_3) (CallFunction|GetItemNames _asbp_station_prop_funding_station (LocalVariables|GetIsMultiChoice)))
          (Utilities|Array|Add _ioedata_3 _conversationnodeparameterpair_0_3)
          (Utilities|Array|Add _ioedata_3 _conversationnodeparameterpair_1_3))
        (:then_5
          (bind _conversationnodeparameterpair_3 (CallFunction|GetNPCIcon _asbp_station_prop_funding_station))
          (Utilities|Array|Add (LocalVariables|GetIOEData) _conversationnodeparameterpair_3))
        (:then_6
          (bind _conversationnodeparameterpair_4 (CallFunction|GetStationIndex _asbp_station_prop_funding_station))
          (Utilities|Array|Add (LocalVariables|GetIOEData) _conversationnodeparameterpair_4))
        (:then_7
          _)))
    (:CastFailed
      _))
  (return (LocalVariables|GetIOEData)))

(fn GetItemIcons
  (params (FundingStation :type BP%20Station%20Prop%20Funding%20Station%20Object%20Reference) (IsMultiChoice :type Boolean))
  (bind _parameterkey (LocalVariables|GetParameterKey))
  (bind _returnvalue (Utilities|String|Append _parameterkey "1"))
  (bind _parameterkey_2 (LocalVariables|GetParameterKey))
  (bind _returnvalue_2 (Utilities|String|Append _parameterkey_2 "0"))
  (bind _fundingreferencea (Variables|/CorruptionGameplay/Gameplay/WarEffort/Prototype/BP_StationProp_FundingStation.BP_StationProp_FundingStation_C|GetFundingReferenceA FundingStation))
  (bind _fundingreferenceb (Variables|/CorruptionGameplay/Gameplay/WarEffort/Prototype/BP_StationProp_FundingStation.BP_StationProp_FundingStation_C|GetFundingReferenceB FundingStation))
  (if (Utilities|IsValidSoftObjectReference _fundingreferencea)
    (LocalVariables|SetToReturn_0 (Utilities|Struct|MakeConversationNodeParameterPair _returnvalue_2 (Utilities|ToString(SoftObjectReference) _fundingreferencea)))
    (else
      (LocalVariables|SetToReturn_0 (Utilities|Struct|MakeConversationNodeParameterPair _returnvalue_2))))
  (if IsMultiChoice
    (if (Utilities|IsValidSoftObjectReference _fundingreferenceb)
      (LocalVariables|SetToReturn_1 (Utilities|Struct|MakeConversationNodeParameterPair _returnvalue (Utilities|ToString(SoftObjectReference) _fundingreferenceb)))
      (else
        (LocalVariables|SetToReturn_1 (Utilities|Struct|MakeConversationNodeParameterPair _returnvalue))))
    (return (LocalVariables|GetToReturn_0) (LocalVariables|GetToReturn_1))))

(fn GetItemNames
  (params (FundingStation :type BP%20Station%20Prop%20Funding%20Station%20Object%20Reference) (IsMultiChoice :type Boolean))
  (return (Utilities|Struct|MakeConversationNodeParameterPair (Utilities|String|Append (LocalVariables|GetParameterKey) "0") (WarEffort|WriteTexttoBuffer (Variables|/CorruptionGameplay/Gameplay/WarEffort/Prototype/BP_StationProp_FundingStation.BP_StationProp_FundingStation_C|GetFundingTextA FundingStation))) (Utilities|Struct|MakeConversationNodeParameterPair (Utilities|String|Append (LocalVariables|GetParameterKey) "1") (WarEffort|WriteTexttoBuffer (Variables|/CorruptionGameplay/Gameplay/WarEffort/Prototype/BP_StationProp_FundingStation.BP_StationProp_FundingStation_C|GetFundingTextB FundingStation)))))

(fn GetCurrentFundingAmount
  (params (FundingStation :type BP%20Station%20Prop%20Funding%20Station%20Object%20Reference) (IsMultipleChoice :type Boolean))
  (bind _funding_station_index (Class|BPStationPropFundingStation|GetFundingStationIndex FundingStation))
  (bind _fundingmanagerref (Variables|/CorruptionGameplay/Gameplay/WarEffort/Prototype/BP_StationProp_FundingStation.BP_StationProp_FundingStation_C|GetFundingManagerRef FundingStation))
  (bind _currentfunds (Variables|/CorruptionGameplay/Gameplay/WarEffort/MeshNetwork/BP_FundingManager_Parent.BP_FundingManager_Parent_C|GetCurrentFunds _fundingmanagerref))
  (bind _output (Utilities|Array|Get(acopy) _currentfunds _funding_station_index))
  (bind _returnvalue (Utilities|Text|ToText(Integer64) _output))
  (bind _returnvalue_2 (Utilities|String|ToString(Text) _returnvalue))
  (bind _parameterkey (LocalVariables|GetParameterKey))
  (bind _returnvalue_3 (Utilities|String|Append _parameterkey "0"))
  (bind _conversationnodeparameterpair (Utilities|Struct|MakeConversationNodeParameterPair _returnvalue_3 _returnvalue_2))
  (if IsMultipleChoice
    (return _conversationnodeparameterpair (Utilities|Struct|MakeConversationNodeParameterPair (Utilities|String|Append (LocalVariables|GetParameterKey) "1") (Utilities|String|ToString(Text) (Utilities|Text|ToText(Integer64) (Utilities|Array|Get(acopy) _currentfunds (+ _funding_station_index 1))))))
    (else
      (return _conversationnodeparameterpair (Utilities|Struct|MakeConversationNodeParameterPair (Utilities|String|Append (LocalVariables|GetParameterKey) "1"))))))

(fn GetIsSingleOrMultiEntry
  (params (FundingStation :type BP%20Station%20Prop%20Funding%20Station%20Object%20Reference))
  (Utilities|Casting|CastToBP_StationProp_FundingStation_MultiChoice FundingStation
    (:then
      (return (Utilities|Struct|MakeConversationNodeParameterPair (LocalVariables|GetParameterKey) (Utilities|String|ToString(Boolean) true)) true))
    (:CastFailed
      (return (Utilities|Struct|MakeConversationNodeParameterPair (LocalVariables|GetParameterKey) (Utilities|String|ToString(Boolean) false)) false))))

(fn GetMaxRequiredFundingAmount
  (params (FundingStation :type BP%20Station%20Prop%20Funding%20Station%20Object%20Reference))
  (return (Utilities|Struct|MakeConversationNodeParameterPair (LocalVariables|GetParameterKey) (Utilities|String|ToString(Float) (ScalableFloat|GetValueatLevel (Variables|/CorruptionGameplay/Gameplay/WarEffort/Prototype/BP_StationProp_FundingStation.BP_StationProp_FundingStation_C|GetMaxFundsRequired FundingStation))))))

(fn GetNPCIcon
  (params (FundingStation :type BP%20Station%20Prop%20Funding%20Station%20Object%20Reference))
  (return (Utilities|Struct|MakeConversationNodeParameterPair (LocalVariables|GetParameterKey) (Utilities|ToString(SoftObjectReference) (LocalVariables|GetNPCIcon)))))

(fn GetStationIndex
  (params (FundingStation :type BP%20Station%20Prop%20Funding%20Station%20Object%20Reference))
  (return (Utilities|Struct|MakeConversationNodeParameterPair (LocalVariables|GetParameterKey) (Utilities|String|ToString(Integer) (Class|BPStationPropFundingStation|GetFundingStationIndex FundingStation)))))
